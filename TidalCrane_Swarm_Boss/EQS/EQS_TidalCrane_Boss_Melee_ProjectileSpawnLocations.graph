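# EQS query that picks projectile spawn locations for the TidalCrane swarm boss's melee variant.
# A spiral donut generator lays rings of points around the querier, projected with a nav filter that excludes smashables and water.
# A nearby-buildings test filters points, then a height-offset trace from the context to each point filters and scores them.
# A random test also filters and scores, varying the chosen locations.
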EnvQuery /TidalCrane_Swarm_Boss/EQS/EQS_TidalCrane_Boss_Melee_ProjectileSpawnLocations
options [{
  generator {
   name=EnvQueryGenerator_Donut_1
   class=EnvQueryGenerator_Donut
   properties {
    bUseSpiralPattern=true
    center=&EnvQueryContext_Querier
    innerRadius=1500
    numberOfRings=4
    outerRadius=4000
    pointsPerRing=12
    projectionData {
     extentX=0
     navigationFilter=&NavFilter_Phoebe_NoSmashableNoWater_C
     postProjectionVerticalOffset=0
     projectDown=1000
     serializedChannel=ECC_GameTraceChannel11
     traceChannel=TraceTypeQuery10
    }
   }
  }
  tests [
   FortQueryTest_HasNearbyBuildings_1 {class=FortQueryTest_HasNearbyBuildings, properties {extentXY=1, testPurpose=Filter}}
   EnvQueryTest_Trace_1 {
    class=EnvQueryTest_Trace
    properties {
     contextHeightOffset=200
     itemHeightOffset=250
     scoringFactor=1
     testComment=""
     testPurpose=FilterAndScore
     traceFromContext=true
     traceData {bOnlyBlockingHits=false, serializedChannel=ECC_GameTraceChannel11, traceChannel=TraceTypeQuery10}
    }
   }
   FortQueryTest_Random_1 {class=FortQueryTest_Random, properties {floatValueMax=1, scoringFactor=1, testPurpose=FilterAndScore}}
  ]
 }]
